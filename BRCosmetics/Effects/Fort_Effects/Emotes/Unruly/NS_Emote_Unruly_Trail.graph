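# Niagara system for the Unruly emote: a warm, glowing trail with flare and sparkles attached to a skeletal mesh.
# Flare is a burst sprite that follows the mesh with a flickering light.
# Two GPU sparkle emitters spawn from sphere and cylinder offsets, thinning over the loop while twinkling and drifting on curl noise.
# An anim-trail ribbon built from mesh bone indices narrows and darkens from warm orange while fading.
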NiagaraSystem /BRCosmetics/Effects/Fort_Effects/Emotes/Unruly/NS_Emote_Unruly_Trail
dependencies {
 usedDataInterfaces [&NiagaraDataInterfaceCurve, &NiagaraDataInterfaceColorCurve]
 usedDynamicInputs []
 usedModules [
  &self:SystemSpawnScript.SystemScriptSource.SystemScriptGraph.NiagaraNodeAssignment_0.SetVariables_#1
  &SystemState
 ]
}
emitters {
 Flare {
  emitterUpdateScript [
   &EmitterState()
   &SpawnBurst_Instantaneous(Use Spawn Probability=false, UseLoopCountLimit=false)
  ]
  particleSpawnScript [
   &InitializeParticle(
    Color=<rgba 8 4.3771 0.7956 1>
    Lifetime=1.7
    Sprite Size Mode=::NewEnumerator3
    Uniform Sprite Size=20
    UsePositionOffset=false
    Write Material Random=true
   )
   &CameraOffset(Camera Offset Amount=5)
  ]
  particleUpdateScript [
   &SkeletalMeshLocation(Bone Sampling Mode=::NewEnumerator4, MaskSpawnGroup=false, Skeletal Mesh=$System.Skeletal Mesh)
   &ParticleState()
   &ScaleSpriteSize(Uniform Curve Sprite Scale=DataInterface() {propertyValues {Curve {keys [<at 0 0 RCIM_Cubic>, {arriveTangent=1.3333, leaveTangent=1.3333, time=0.2, value=1}, <at 0.75 1>]}}})
   &ScaleColor(
    Scale RGB=&Multiply_VectorByFloat(
     Vector=<1 1 1>
     Float=&FloatFromCurve(
      FloatCurve=DataInterface() {propertyValues {Curve {keys [{interpMode=RCIM_Cubic, tangentMode=RCTM_User, time=0, value=1}, {interpMode=RCIM_Cubic, tangentMode=RCTM_User, time=1, value=0}]}}}
      Scale Curve=&Waveform("Waveform [1]"=::NewEnumerator12, "[1] Frequency"=15.9, "[1] Amplitude Min/Max"=<0 1>, "[1] Bias"=0.8)
     )
    )
    ScaleA=true
    ScaleRGB=true
   )
   /SetVariables(Particles.Flaresize2:Vector2f=&Multiply_Vector2D(A=$Particles.SpriteSize, B=<1.5149 0.25>)) {isSetParameters=true}
   &FN_Light_Attributes(Create Exponent Attribute=true, Create Radius Attribute=true, Exponent=4, Radius=35)
  ]
  rendererClasses [&NiagaraSpriteRendererProperties, &NiagaraLightRendererProperties]
  renderers [
   {rendererClass=&NiagaraSpriteRendererProperties}
   {rendererClass=&NiagaraSpriteRendererProperties, rendererIndex=1}
   {rendererClass=&NiagaraLightRendererProperties, rendererIndex=2}
  ]
 }
 Sparkles {
  simTarget=GPUComputeSim
  emitterUpdateScript [
   &EmitterState()
   &SpawnRate(
    Spawn Probability=&FloatFromCurve(FloatCurve=DataInterface() {propertyValues {Curve {keys [<at 0 1>,<at 1 0>]}}}, CurveIndex=$Emitter.NormalizedLoopAge)
    SpawnRate=40
    Use Spawn Probability=true
   )
  ]
  particleSpawnScript [
   &InitializeParticle(
    Color=<rgba 5 3.7043 1.51 1>
    Lifetime Max=0.5
    Lifetime Min=0.2
    Lifetime Mode=::NewEnumerator1
    Sprite Size Mode=::NewEnumerator1
    UsePositionOffset=false
    Write Material Random=true
   )
   &SkeletalMeshLocation(Bone Sampling Mode=::NewEnumerator4, MaskSpawnGroup=false, Skeletal Mesh=$System.Skeletal Mesh)
   &ShapeLocation(MaskSpawnGroup=false, Override Seed=false, Sphere Radius=5, UseOwnerScale=false)
   &CameraOffset(Camera Offset Amount=10)
  ]
  particleUpdateScript [
   &ParticleState()
   &ScaleSpriteSize(
    Uniform Curve Scale=&Waveform("Waveform [1]"=::NewEnumerator9, "[1] Frequency"=12, "[1] Amplitude Min/Max"=<0 1>, "[1] Bias"=0.5)
    Uniform Curve Sprite Scale=DataInterface() {propertyValues {Curve {keys [
        {arriveTangent=9.0725, interpMode=RCIM_Cubic, leaveTangent=9.0725, tangentMode=RCTM_User, time=0, value=0}
        {arriveTangent=-0.1942, leaveTangent=-0.1942, time=0.15, value=1}
        {arriveTangent=-1.1765, interpMode=RCIM_Cubic, leaveTangent=-1.1765, time=0.7, value=1}
        {interpMode=RCIM_Cubic, leaveTangent=-0.0405, time=1, value=0}
       ]}}}
   )
   &Drag(Drag=&RandomRangeFloat(Minimum=10, Maximum=12), Use Linear Drag=true, Use Rotational Drag=true)
   &InheritVelocity(Speed Limit=true, UseSpeedThreshold=false)
   &CurlNoiseForce(Noise Strength=500, Pan Noise=false)
   &SolveForcesAndVelocity(Clamp Velocity=false, Limit Acceleration=false)
   &ScaleColor(
    Scale Alpha=&FloatFromCurve(FloatCurve=DataInterface() {propertyValues {Curve {keys [<at 0 1>,<at 1 0>]}}})
    Scale RGB=&Multiply_VectorByFloat(
     Vector=<1 1 1>
     Float=&Waveform(
      "Waveform [1]"=::NewEnumerator9
      "[1] Phase Scale"=2
      "[1] Frequency"=&RandomRangeFloat(Minimum=10, Maximum=20)
      "[1] Amplitude Min/Max"=<0 1>
     )
    )
    ScaleA=true
    ScaleRGB=true
   )
  ]
  rendererClasses [&NiagaraSpriteRendererProperties]
  renderers [{rendererClass=&NiagaraSpriteRendererProperties}]
 }
 SparklesFeather {
  simTarget=GPUComputeSim
  emitterUpdateScript [
   &EmitterState()
   &SpawnRate(
    Spawn Probability=&FloatFromCurve(FloatCurve=DataInterface() {propertyValues {Curve {keys [<at 0 1>,<at 1 0>]}}}, CurveIndex=$Emitter.NormalizedLoopAge)
    SpawnRate=40
    Use Spawn Probability=true
   )
  ]
  particleSpawnScript [
   &InitializeParticle(
    Color=<rgba 5 3.7043 1.51 1>
    Lifetime Max=0.5
    Lifetime Min=0.2
    Lifetime Mode=::NewEnumerator1
    Sprite Size Mode=::NewEnumerator1
    UsePositionOffset=false
    Write Material Random=true
   )
   &SkeletalMeshLocation(Bone Sampling Mode=::NewEnumerator4, MaskSpawnGroup=false, Skeletal Mesh=$System.Skeletal Mesh)
   &ShapeLocation(
    Cylinder Height=50
    Cylinder Radius=1
    MaskSpawnGroup=false
    Override Seed=false
    Shape Primitive=::NewEnumerator1
    Surface Only=false
    UseOwnerScale=false
   )
   &CameraOffset(Camera Offset Amount=10)
  ]
  particleUpdateScript [
   &ParticleState()
   &ScaleSpriteSize(
    Uniform Curve Scale=&Waveform("Waveform [1]"=::NewEnumerator9, "[1] Frequency"=12, "[1] Amplitude Min/Max"=<0 1>, "[1] Bias"=0.5)
    Uniform Curve Sprite Scale=DataInterface() {propertyValues {Curve {keys [
        {arriveTangent=9.0725, interpMode=RCIM_Cubic, leaveTangent=9.0725, tangentMode=RCTM_User, time=0, value=0}
        {arriveTangent=-0.1942, leaveTangent=-0.1942, time=0.15, value=1}
        {arriveTangent=-1.1765, interpMode=RCIM_Cubic, leaveTangent=-1.1765, time=0.7, value=1}
        {interpMode=RCIM_Cubic, leaveTangent=-0.0405, time=1, value=0}
       ]}}}
   )
   &Drag(Drag=&RandomRangeFloat(Minimum=10, Maximum=12), Use Linear Drag=true, Use Rotational Drag=true)
   &InheritVelocity(Speed Limit=true, UseSpeedThreshold=false)
   &CurlNoiseForce(Noise Strength=500, Pan Noise=false)
   &SolveForcesAndVelocity(Clamp Velocity=false, Limit Acceleration=false)
   &ScaleColor(
    Scale Alpha=&FloatFromCurve(FloatCurve=DataInterface() {propertyValues {Curve {keys [<at 0 1>,<at 1 0>]}}})
    Scale RGB=&Multiply_VectorByFloat(
     Vector=<1 1 1>
     Float=&Waveform(
      "Waveform [1]"=::NewEnumerator9
      "[1] Phase Scale"=2
      "[1] Frequency"=&RandomRangeFloat(Minimum=10, Maximum=20)
      "[1] Amplitude Min/Max"=<0 1>
     )
    )
    ScaleA=true
    ScaleRGB=true
   )
  ]
  rendererClasses [&NiagaraSpriteRendererProperties]
  renderers [{rendererClass=&NiagaraSpriteRendererProperties}]
 }
 Trail {
  emitterUpdateScript [
   &EmitterState()
   &FN_SpawnAnimTrail(Spawn Interval=2, Spawn Mode:ENiagaraSkeletalSpawning=::NewEnumerator0)
  ]
  particleSpawnScript [
   &FN_InitializeAnimTrailBoneIndecies()
   &FN_InitializeAnimTrail(AnimTrailWidth=12, Color=<rgba 1 1 1 1.5>, SkeletalMesh=$System.Skeletal Mesh)
  ]
  particleUpdateScript [
   &SolveForcesAndVelocity(Clamp Velocity=false, Limit Acceleration=false)
   &FN_UpdateAnimTrail(Locked Particle RibbonWidth=1, SkeletalMesh=$User.SkeletalMesh) {module=FN_UpdateAnimTrail001}
   &ScaleColor(
    Scale Alpha=&FloatFromCurve(FloatCurve=DataInterface() {propertyValues {Curve {keys [<at 0 1>,<at 1 0>]}}})
    Scale RGB=&MakeVectorFromLinearColorRGB(Color=&ColorFromCurve(ColorCurve=DataInterface() {propertyValues {
        AlphaCurve {keys [<at 0 1 RCIM_Cubic>]}
        BlueCurve {keys [<at 0 0.1663 RCIM_Cubic>, {arriveTangent=-0.0592, interpMode=RCIM_Cubic, leaveTangent=-0.0592, time=0.4034, value=0.0891}, <at 1 0.1071>]}
        GreenCurve {keys [<at 0 1 RCIM_Cubic>, {arriveTangent=-0.8385, interpMode=RCIM_Cubic, leaveTangent=-0.8385, time=0.4034, value=0.75}, <at 1 0.1615>]}
        RedCurve {keys [<at 0 0.1521 RCIM_Cubic>, {arriveTangent=-0.1362, interpMode=RCIM_Cubic, leaveTangent=-0.1362, time=0.4034, value=0.0779}, <at 1 0.0159>]}
       }}))
    ScaleA=true
    ScaleRGB=true
   ) {enabled=false}
   &ScaleColor(
    Scale Alpha=&FloatFromCurve(FloatCurve=DataInterface() {propertyValues {Curve {keys [<at 0 1>,<at 1 0>]}}})
    Scale RGB=&MakeVectorFromLinearColorRGB(Color=&ColorFromCurve(ColorCurve=DataInterface() {propertyValues {
        AlphaCurve {keys [<at 0 1 RCIM_Cubic>]}
        BlueCurve {keys [<at 0 0.1521 RCIM_Cubic>, {arriveTangent=-0.1365, interpMode=RCIM_Cubic, leaveTangent=-0.1365, time=0.4034, value=0.0779}, <at 1 0.0156>]}
        GreenCurve {keys [<at 0 0.5231 RCIM_Cubic>, {arriveTangent=-0.4766, interpMode=RCIM_Cubic, leaveTangent=-0.4766, time=0.4034, value=0.2985}, <at 1 0.0465>]}
        RedCurve {keys [<at 0 1 RCIM_Cubic>, {arriveTangent=-0.7188, interpMode=RCIM_Cubic, leaveTangent=-0.7188, time=0.4034, value=0.75}, <at 1 0.2812>]}
       }}))
    ScaleA=true
    ScaleRGB=true
   ) {module=ScaleColor001}
   &ScaleRibbonWidth(Ribbon Width Scale=&FloatFromCurve(FloatCurve=DataInterface() {propertyValues {Curve {keys [<at 0.4 0.95>,<at 1 0>]}}}))
  ]
  rendererClasses [&NiagaraRibbonRendererProperties]
  renderers [{rendererClass=&NiagaraRibbonRendererProperties}]
 }
}
summary {
 systemName=NS_Emote_Unruly_Trail
 emitters [
  {emitterName=Flare, rendererClasses [&NiagaraSpriteRendererProperties, &NiagaraLightRendererProperties]}
  {emitterName=Sparkles, simTarget=GPUComputeSim, rendererClasses [&NiagaraSpriteRendererProperties]}
  {emitterName=SparklesFeather, simTarget=GPUComputeSim, rendererClasses [&NiagaraSpriteRendererProperties]}
  {emitterName=Trail, rendererClasses [&NiagaraRibbonRendererProperties]}
 ]
 userVariables [
  User.RibbonIDOffset {type=NiagaraInt32}
  User.SkeletalMesh {
   type=NiagaraDataInterfaceSkeletalMesh
   defaultValue {
    struct=NiagaraExt_VariableValue_DataInterface
    value {dataInterface=&self:NiagaraDataInterfaceSkeletalMesh_2}
   }
  }
 ]
}
